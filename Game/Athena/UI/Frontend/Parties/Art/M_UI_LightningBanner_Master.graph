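# UI master material for the frontend party lightning banner, drawing either a procedural banner body or a textured tail via static switches.
# The body multiplies four smoothstepped, rotated-UV edges into a skewed quad mask, with ShapeChange morphing between two edge sets.
# Its color mixes a gradient with a secondary color, while ColorChange blends to alternate palettes and s_WhiteFill brightens.
# The tail tints texture red and takes opacity from alpha, optionally fading the top.
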Material /Game/Athena/UI/Frontend/Parties/Art/M_UI_LightningBanner_Master
graph {
 properties {
  bUsedWithStaticMesh=false
  blendMode=BLEND_Translucent
  materialDomain=MD_UI
  pixelDepthOffsetMode=PDOM_Legacy
  usedShadingModels=Default Lit
 }
 nodes [
  Constant() {r=0.01}
  Constant() {r=0.505}
  Constant()
  LinearInterpolate()
  PerInstanceRandom()
  n1 = MaterialFunctionCall() {materialFunction=&GetUserInterfaceUV}
  n2 = ComponentMask(None=@n1.Normalized UV) {g=true}
  n3 = StaticSwitchParameter(
   True=Add(A=Power(Base=@n2), B=ScalarParameter() {defaultValue=0.6, parameterName=YGradOffset})
   False=ComponentMask(None=@n1.Normalized UV) {r=true}
  ) {parameterName=UseVerticalGradient}
  n4 = LinearInterpolate(
   A=VectorParameter() {parameterName=TopGradient_L, defaultValue=<rgba 0.0533 0.2205 0.4375 1>}.RGB
   B=VectorParameter() {parameterName=TopGradient_R, defaultValue=<rgba 0.1476 0.3187 0.7083 1>}.RGB
   Alpha=@n3
  )
  n5 = LinearInterpolate(
   A=VectorParameter() {parameterName=Alt_TopGradient_L, defaultValue=<rgba 0.5417 0.5417 0.5417 1>}.RGB
   B=VectorParameter() {parameterName=Alt_TopGradient_R, defaultValue=<rgba 0.6823 0.6823 0.6823 1>}.RGB
   Alpha=@n3
  )
  n6 = ScalarParameter() {defaultValue=1, parameterName=ColorChange}
  n7 = LinearInterpolate(
   A=VectorParameter() {parameterName=RGB_2, defaultValue=<rgba 0 0.1052 0.3854 1>}.RGB
   B=VectorParameter() {parameterName=RGB_Alt_2, defaultValue=<rgba 0.1979 0.1979 0.1979 1>}.RGB
   Alpha=@n6
  )
  n8 = Rotator(Coordinate=@n1.Normalized UV, Time=Divide(A=Constant() {r=-30}) {constA=360, constB=360}) {speed=6.2832}
  n9 = Rotator(Coordinate=@n1.Normalized UV, Time=Divide(A=Constant() {r=30}) {constA=360, constB=360}) {speed=6.2832}
  n10 = MaterialFunctionCall(Alpha=DotProduct(A=ComponentMask(None=@n8) {g=true}, B=ComponentMask(None=@n9) {g=true}), Min=Constant() {r=0.24}, Max=Constant() {r=0.26}) {materialFunction=&SmoothStep}
  n11 = ComponentMask(None=TextureCoordinate()) {r=true}
  n12 = ScalarParameter() {parameterName=s_WhiteFill, sliderMax=1}
  n13 = Add(
   A=LinearInterpolate(
    A=LinearInterpolate(A=@n4, B=@n5, Alpha=@n6)
    B=@n7
    Alpha=StaticSwitchParameter(True=@n10.Result, False=@n11) {defaultValue=true, parameterName=UseHorizonGrad}
   ) {constB=0}
   B=@n12
  )
  n14 = TextureCoordinate()
  n15 = ScalarParameter() {defaultValue=0.005, parameterName=R_R_A}
  n16 = ScalarParameter() {defaultValue=-0.005, parameterName=R_R_B}
  n17 = Clamp(None=ScalarParameter() {parameterName=ShapeChange}) {maxDefault=1.2}
  n18 = ComponentMask(None=Rotator(Coordinate=@n14, Time=LinearInterpolate(A=@n15, B=@n16, Alpha=@n17) {constA=0.005, constB=-0.005}) {speed=6.2832}) {r=true}
  n19 = Constant() {r=0.05}
  n20 = Subtract(A=@n19) {constB=0.0035}
  n21 = ScalarParameter() {defaultValue=0.49, parameterName=R_L_A}
  n22 = ScalarParameter() {defaultValue=0.51, parameterName=R_L_B}
  n23 = ComponentMask(None=Rotator(Coordinate=@n14, Time=LinearInterpolate(A=@n21, B=@n22, Alpha=@n17) {constA=0.49, constB=0.51}) {speed=6.2832}) {r=true}
  n24 = Constant() {r=0.05}
  n25 = Subtract(A=@n24) {constB=0.0035}
  n26 = LinearInterpolate(
   A=ScalarParameter() {defaultValue=0.1, parameterName=V_UpperBound_Subtract_A}
   B=ScalarParameter() {defaultValue=0.13, parameterName=V_UpperBound_Subtract_B}
   Alpha=@n17
  ) {constA=0.1, constB=0.13}
  n27 = ScalarParameter() {parameterName=V_R_A}
  n28 = ScalarParameter() {defaultValue=0.002, parameterName=V_R_B}
  n29 = ComponentMask(None=Rotator(Coordinate=Subtract(A=@n14, B=@n26), Time=LinearInterpolate(A=@n27, B=@n28, Alpha=@n17) {constB=0.002}) {speed=6.2832}) {g=true}
  n30 = Constant() {r=0.05}
  n31 = Subtract(A=@n30) {constB=0.015}
  n32 = ScalarParameter() {defaultValue=0.1, parameterName=V_LowerBound_Subtract_A}
  n33 = ScalarParameter() {defaultValue=0.49, parameterName=V_L_A}
  n34 = ScalarParameter() {defaultValue=0.505, parameterName=V_L_B}
  n35 = ComponentMask(None=Rotator(Coordinate=Add(A=@n14, B=@n32), Time=LinearInterpolate(A=@n33, B=@n34, Alpha=@n17) {constA=0.49, constB=0.505}) {speed=6.2832}) {g=true}
  n36 = Constant() {r=0.05}
  n37 = Subtract(A=@n36) {constB=0.015}
  n38 = DotProduct(
   A=DotProduct(
    A=MaterialFunctionCall(Alpha=@n18, Min=@n20, Max=@n19) {materialFunction=&SmoothStep}.Result
    B=MaterialFunctionCall(Alpha=@n23, Min=@n25, Max=@n24) {materialFunction=&SmoothStep}.Result
   )
   B=DotProduct(
    A=MaterialFunctionCall(Alpha=@n29, Min=@n31, Max=@n30) {materialFunction=&SmoothStep}.Result
    B=MaterialFunctionCall(Alpha=@n35, Min=@n37, Max=@n36) {materialFunction=&SmoothStep}.Result
   )
  )
  n39 = TextureSampleParameter2D() {parameterName=Texture, texture=&/T_UI_LightningBanner_Tail}
  n40 = Multiply(A=@n7, B=@n39.R)
  StaticSwitchParameter(True=Multiply(A=Clamp(None=@n13), B=@n38) {constA=1}, False=Clamp(None=Add(A=@n40, B=@n12))) {defaultValue=true, parameterName=MainBorder}
  n41 = ComponentMask(None=TextureCoordinate()) {g=true}
  n42 = Clamp(None=Multiply(A=Subtract(A=@n41) {constB=0.1}) {constB=10})
  StaticSwitchParameter(True=@n38, False=StaticSwitchParameter(True=Multiply(A=@n42, B=@n39.A), False=@n39.A) {parameterName=AlphaBlendTop}) {defaultValue=true, parameterName=MainBorder_1}
 ]
}
